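; Quest gameplay ability for S11_OT3_Q05 that credits the player for being in a no-swimming area while swimming.
; On event activation it runs the area check immediately, then repeats it on a looping timer at the progress interval.
; Each check tests the owning pawn for the no-swimming tag and, on a match, sends a custom stat event to the quest manager.
; Removal of the swimming tag clears the timer and ends the ability.
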(event Ability|EventActivateAbilityFromEvent
  (params (EventData :type Gameplay%20Event%20Data%20Structure%20%28by%20ref%29))
  (|CallParentFunction EventData)
  (CallFunction|CheckIfOverlappedNoSwimmingArea)
  (bind _returnvalue (Utilities|Time|SetTimerbyEvent (delegate Custom|Check%20if%20overlapped%20no%20swimming%20area) (Variables|GAVars|Setup|GetProgress%20Interval) true))
  (Variables|GAVars|Data|SetTimer_Progress _returnvalue)
  (Ability|Tasks|WaitGameplayTagRemove (Variables|GAVars|Setup|GetT_Swimming)
    (:Removed
      (Utilities|Time|ClearandInvalidateTimerbyHandle (Variables|GAVars|Data|GetTimer_Progress))
      (Ability|EndAbility))))

(event Custom|Check%20if%20overlapped%20no%20swimming%20area
  (if (GameplayTags|TagContainer|CompareTagAssetInterfacetoOtherTags (Variables|GAVars|Setup|GetT_NoSwimming) (Variables|GAVars|Data|GetOwningPlayerPawn))
    (Quest|SendComplexCustomStatEvent (Variables|GAVars|Data|GetOwningPlayerQuestManager) 0 (Variables|GAVars|Setup|GetTC_Quest) "(GameplayTags=)" 1)))
